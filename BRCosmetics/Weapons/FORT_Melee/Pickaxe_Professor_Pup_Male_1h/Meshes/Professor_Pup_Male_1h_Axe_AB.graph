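# Animation Blueprint for the Professor Pup one-handed pickaxe that continuously spins its dyn_saw bone.
# Each update, Spin Function advances an angle by speed times deltaTimeRemap plus velocityMult, wrapping at 720.
# The angle becomes a pitch rotator, applied additively in bone space through a Transform (Modify) Bone node.
# The deltaTimeRemap interpolation receives zero delta time, so the spin steps per update rather than by elapsed time.
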AnimBlueprint /BRCosmetics/Weapons/FORT_Melee/Pickaxe_Professor_Pup_Male_1h/Meshes/Professor_Pup_Male_1h_Axe_AB
target_skeleton=&/Professor_Pup_Male_1h_Axe_Skeleton
graphs [
 AnimGraph {nodes [
   n1 = "Transform (Modify) Bone - Bone: dyn_saw"(
    ComponentPose=Local To Component(LocalPose=Input Pose())
    Alpha=1
    Translation="0,0,0"
    Rotation=Get spinResultA() -> (spinResultA:Rotator)
    Scale="1,1,1"
   ) {
    rotationMode=BMM_Additive
    rotationSpace=BCS_BoneSpace
    alphaBoolBlend {blendOption=Linear}
    alphaScaleBiasClamp {bInterpResult=true}
    boneToModify {boneName=dyn_saw}
   }
   Output Pose(Result=Component To Local(ComponentPose=@n1)) {layerGroup=None}
  ]}
 EventGraph {nodes [
   Event BlueprintUpdateAnimation()
   > n1 = Sequence!(execute) -> (then_0, then_1, then_2:Exec)
   > Set deltaTimeRemap(deltaTimeRemap:double=FInterpTo(Current=Get deltaTimeRemap() -> (deltaTimeRemap:double), Target=1, DeltaTime=0, InterpSpeed=1)) -> (Output_Get:double)
   n2 = Spin Function!(execute=@n1.then_1, self, Spin:double=Get spinA() -> (spinA:double), Speed:double=Get speedA() -> (speedA:double)) -> (then, Output:double)
   > n3 = Set spinA(spinA:double=@n2) -> (Output_Get:double)
   > Set spinResultA(spinResultA:Rotator=MakeRotator(Pitch=@n3)) -> (Output_Get:Rotator)
   TryGetPawnOwner()
  ]}
 Spin Function {nodes [
   n1 = Spin Function!() -> (then, Spin:double=~, Speed:double=~) {is=title K2Node_FunctionEntry}
   > n2 = Set spinVar(spinVar:double=@n1) -> (Output_Get:double)
   > n3 = Branch(Condition="float >= float"(A=@n2, B=720) {is=title K2Node_CallFunction})
   > n4 = Set spinVar(spinVar=float - float(A=Get spinVar() -> (spinVar:double), B=720) {is=title K2Node_CallFunction})
   n5 = Get velocityMult() -> (velocityMult:double)
   n6 = Get deltaTimeRemap()
   n7 = Set spinVar(
    execute=@n3.else
    spinVar=float + float!(A=Get spinVar(), B=float * float(A=@n1.Speed, B=@n6), C:double=float * float(A=1, B=@n5)) -> (ReturnValue) {is=title K2Node_CommutativeAssociativeBinaryOperator}
   )
   Return Node!(execute=[@n4.then, @n7.then], Output:double=@n7)
  ]}
]
variables [
 deltaTimeRemap {type=number}
 speedA {type=number, default=5}
 spinA {type=number}
 spinResultA {type=Rotator}
 spinVar {type=number}
 velocityMult {type=number, default=12}
]
